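# First-person sound cue for firing the Shockwave bow at full tension.
# It mixes a random metal bow release, reused from the Metal bow, with a random Shockwave whoosh that follows after a short fixed delay.
# Each layer passes through its own modulator, with the whoosh pitch randomized upward.
# It uses first-person projectile attenuation, the ranged-weapon fire sound class and first-person gunfire concurrency.
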SoundCue /PrimalGameplay/Sounds/Weapons/Bows/Shockwave/Bow_Shockwave_Fire_Full_Tension_1P_Cue
root=Mixer_0
nodes [
 {id=Delay_1, children [Modulator_3], properties {delayMax=0.16, delayMin=0.16}}
 {id=Mixer_0, children [Modulator_2,Delay_1], properties {inputVolume [0.6,0.6]}}
 {id=Modulator_2, children [Random_0]}
 {id=Modulator_3, children [Random_2], properties {pitchMax=1.25, pitchMin=1}}
 {id=Random_0, children [WavePlayer_1,WavePlayer_2,WavePlayer_3], properties {weights [1,1,1]}}
 {id=Random_2, children [WavePlayer_7, WavePlayer_8, WavePlayer_9, WavePlayer_10], properties {weights [1,1,1,1]}}
 {id=WavePlayer_1, properties {soundWaveAssetPtr=/PrimalGameplay/Sounds/Weapons/Bows/Metal/Bow_Metal_Fire_Full_Tension_01}}
 {id=WavePlayer_10, properties {soundWaveAssetPtr=&/Bow_Shockwave_Fire_Full_Tension_Whoosh_04}}
 {id=WavePlayer_2, properties {soundWaveAssetPtr=/PrimalGameplay/Sounds/Weapons/Bows/Metal/Bow_Metal_Fire_Full_Tension_02}}
 {id=WavePlayer_3, properties {soundWaveAssetPtr=/PrimalGameplay/Sounds/Weapons/Bows/Metal/Bow_Metal_Fire_Full_Tension_03}}
 {id=WavePlayer_7, properties {soundWaveAssetPtr=&/Bow_Shockwave_Fire_Full_Tension_Whoosh_01}}
 {id=WavePlayer_8, properties {soundWaveAssetPtr=&/Bow_Shockwave_Fire_Full_Tension_Whoosh_02}}
 {id=WavePlayer_9, properties {soundWaveAssetPtr=&/Bow_Shockwave_Fire_Full_Tension_Whoosh_03}}
]
properties {
 attenuationSettings=&Fort_Weapon_Projectile_FirstPerson_Athena_Att
 maxDistance=2097152
 soundClassObject=&Fort_WeaponRanged_Fire_1P_PSM
 volumeMultiplier=1
 attenuationOverrides {falloffDistance=5000, attenuationShapeExtents {x=1000}}
 concurrencySet [&Weapon_Gunfire_FirstPerson_Conc]
}
