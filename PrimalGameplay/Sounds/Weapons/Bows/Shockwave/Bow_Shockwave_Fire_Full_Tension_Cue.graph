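# Sound cue for firing the Shockwave bow at full tension, built from a close layer and a distant layer with separate attenuation and third-person or distant ranged-fire sound classes.
# The close layer mixes random metal bow shots with a briefly delayed, pitch-varied random Shockwave whoosh.
# The distant layer randomly plays charged distant shots from the default bow.
# Global and local bow concurrency rules apply.
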SoundCue /PrimalGameplay/Sounds/Weapons/Bows/Shockwave/Bow_Shockwave_Fire_Full_Tension_Cue
root=Mixer_3
nodes [
 {id=Attenuation_0, children [Mixer_0], properties {attenuationSettings=/Game/Sounds/Attenuation/Fort_Weapon_Projectile_Close_Bow_Athena_Att}}
 {
  id=Attenuation_2
  children [Modulator_8]
  properties {attenuationSettings=/Game/Sounds/Attenuation/Fort_Weapon_Projectile_Distant_Bow_Athena_Att}
 }
 {id=Delay_1, children [Modulator_0], properties {delayMax=0.16, delayMin=0.16}}
 {id=Mixer_0, children [Modulator_2,Delay_1], properties {inputVolume [0.6,0.6]}}
 {id=Mixer_3, children [SoundClass_0,SoundClass_4], properties {inputVolume [0.65,0.7]}}
 {id=Modulator_0, children [Random_2], properties {pitchMax=1.25, pitchMin=1}}
 {id=Modulator_2, children [Random_1]}
 {id=Modulator_8, children [Random_0]}
 {id=Random_0, children [WavePlayer_0,WavePlayer_1,WavePlayer_7], properties {weights [1,1,1]}}
 {id=Random_1, children [WavePlayer_2,WavePlayer_3,WavePlayer_4], properties {weights [1,1,1]}}
 {id=Random_2, children [WavePlayer_5, WavePlayer_8, WavePlayer_9, WavePlayer_10], properties {weights [1,1,1,1]}}
 {
  id=SoundClass_0
  children [Attenuation_0]
  properties {soundClassOverride=&Fort_WeaponRanged_Fire_3P_PSM}
 }
 {
  id=SoundClass_4
  children [Attenuation_2]
  properties {soundClassOverride=&Fort_WeaponRanged_Fire_Distant}
 }
 {id=WavePlayer_0, properties {soundWaveAssetPtr=/PrimalGameplay/Sounds/Weapons/Bows/Metal/Bow_Default_Fire_3P_Charged_Distant_01}}
 {id=WavePlayer_1, properties {soundWaveAssetPtr=/PrimalGameplay/Sounds/Weapons/Bows/Metal/Bow_Default_Fire_3P_Charged_Distant_02}}
 {id=WavePlayer_10, properties {soundWaveAssetPtr=&/Bow_Shockwave_Fire_Full_Tension_Whoosh_04}}
 {id=WavePlayer_2, properties {soundWaveAssetPtr=/PrimalGameplay/Sounds/Weapons/Bows/Metal/Bow_Metal_Fire_Full_Tension_01}}
 {id=WavePlayer_3, properties {soundWaveAssetPtr=/PrimalGameplay/Sounds/Weapons/Bows/Metal/Bow_Metal_Fire_Full_Tension_02}}
 {id=WavePlayer_4, properties {soundWaveAssetPtr=/PrimalGameplay/Sounds/Weapons/Bows/Metal/Bow_Metal_Fire_Full_Tension_03}}
 {id=WavePlayer_5, properties {soundWaveAssetPtr=&/Bow_Shockwave_Fire_Full_Tension_Whoosh_01}}
 {id=WavePlayer_7, properties {soundWaveAssetPtr=/PrimalGameplay/Sounds/Weapons/Bows/Metal/Bow_Default_Fire_3P_Charged_Distant_03}}
 {id=WavePlayer_8, properties {soundWaveAssetPtr=&/Bow_Shockwave_Fire_Full_Tension_Whoosh_02}}
 {id=WavePlayer_9, properties {soundWaveAssetPtr=&/Bow_Shockwave_Fire_Full_Tension_Whoosh_03}}
]
properties {
 maxDistance=27500
 volumeMultiplier=1
 attenuationOverrides {falloffDistance=5000, attenuationShapeExtents {x=1000}}
 concurrencySet [/Game/Sounds/Concurrency_Rules/Bow_Global_Conc, /Game/Sounds/Concurrency_Rules/Bow_Local_Conc]
}
