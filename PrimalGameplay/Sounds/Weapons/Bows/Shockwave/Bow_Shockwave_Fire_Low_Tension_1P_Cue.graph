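# First-person sound cue for firing the Shockwave bow at low tension.
# A random node picks one of three equally weighted Metal bow low-tension fire waves, reusing that bow's audio, and passes it through a modulator.
# It plays at reduced volume with first-person projectile attenuation, the first-person ranged fire sound class and gunfire concurrency.
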SoundCue /PrimalGameplay/Sounds/Weapons/Bows/Shockwave/Bow_Shockwave_Fire_Low_Tension_1P_Cue
root=Modulator_2
nodes [
 {id=Modulator_2, children [Random_0]}
 {id=Random_0, children [WavePlayer_0,WavePlayer_1,WavePlayer_2], properties {weights [1,1,1]}}
 {id=WavePlayer_0, properties {soundWaveAssetPtr=/PrimalGameplay/Sounds/Weapons/Bows/Metal/Bow_Metal_Fire_Low_Tension_01}}
 {id=WavePlayer_1, properties {soundWaveAssetPtr=/PrimalGameplay/Sounds/Weapons/Bows/Metal/Bow_Metal_Fire_Low_Tension_02}}
 {id=WavePlayer_2, properties {soundWaveAssetPtr=/PrimalGameplay/Sounds/Weapons/Bows/Metal/Bow_Metal_Fire_Low_Tension_03}}
]
properties {
 attenuationSettings=&Fort_Weapon_Projectile_FirstPerson_Athena_Att
 maxDistance=2097152
 soundClassObject=&Fort_WeaponRanged_Fire_1P_PSM
 volumeMultiplier=0.6
 attenuationOverrides {falloffDistance=5000, attenuationShapeExtents {x=1000}}
 concurrencySet [&Weapon_Gunfire_FirstPerson_Conc]
}
