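# Sound cue for the Shockwave bow's full-tension confirmation, with a sustained loop after the initial confirm sound.
# A mixer layers a pitch-varied confirm sound, a looping bed faded in by an enveloper, and a delayed loop of eight random one-shots with pitch and volume modulation.
# It uses weapon foley class, weapon reload attenuation and global and local bow concurrency rules, with the debug flag left on.
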SoundCue /PrimalGameplay/Sounds/Weapons/Bows/Shockwave/Bow_Shockwave_Full_Tension_Confirm_Cue
root=Mixer_0
nodes [
 {id=Delay_0, children [Modulator_1], properties {delayMax=1.5, delayMin=1}}
 {
  id=Enveloper_0
  children [WavePlayer_2]
  properties {pitchCurve {editorCurveData {keys [{time=0}]}}, volumeCurve {editorCurveData {keys [<at 0 0>,<at 0.5 1>]}}}
 }
 {id=Looping_0, children [Random_0]}
 {id=Mixer_0, children [Modulator_0,Enveloper_0,Delay_0], properties {inputVolume [1.2,0.8,0.8]}}
 {id=Modulator_0, children [WavePlayer_0], properties {pitchMax=1.1, pitchMin=0.9, volumeMax=1, volumeMin=1}}
 {id=Modulator_1, children [Looping_0], properties {pitchMax=1.3, volumeMax=1.5}}
 {
  id=Random_0
  children [WavePlayer_3, WavePlayer_4, WavePlayer_5, WavePlayer_6, WavePlayer_7, WavePlayer_8, WavePlayer_9, WavePlayer_10]
  properties {weights [1,1,1,1,1,1,1,1]}
 }
 {id=WavePlayer_0, properties {soundWaveAssetPtr=&/Bow_Shockwave_Full_Tension_Confirm_01}}
 {
  id=WavePlayer_10
  properties {soundWaveAssetPtr=&/Bow_Shockwave_Full_Tension_Confirm_Loop_One_Shots_08}
 }
 {
  id=WavePlayer_2
  properties {bLooping=true, soundWaveAssetPtr=&/Bow_Shockwave_Full_Tension_Confirm_Loop}
 }
 {
  id=WavePlayer_3
  properties {soundWaveAssetPtr=&/Bow_Shockwave_Full_Tension_Confirm_Loop_One_Shots_01}
 }
 {
  id=WavePlayer_4
  properties {soundWaveAssetPtr=&/Bow_Shockwave_Full_Tension_Confirm_Loop_One_Shots_02}
 }
 {
  id=WavePlayer_5
  properties {soundWaveAssetPtr=&/Bow_Shockwave_Full_Tension_Confirm_Loop_One_Shots_03}
 }
 {
  id=WavePlayer_6
  properties {soundWaveAssetPtr=&/Bow_Shockwave_Full_Tension_Confirm_Loop_One_Shots_04}
 }
 {
  id=WavePlayer_7
  properties {soundWaveAssetPtr=&/Bow_Shockwave_Full_Tension_Confirm_Loop_One_Shots_05}
 }
 {
  id=WavePlayer_8
  properties {soundWaveAssetPtr=&/Bow_Shockwave_Full_Tension_Confirm_Loop_One_Shots_06}
 }
 {
  id=WavePlayer_9
  properties {soundWaveAssetPtr=&/Bow_Shockwave_Full_Tension_Confirm_Loop_One_Shots_07}
 }
]
properties {
 attenuationSettings=&Fort_Weapon_Reload_Att
 bDebug=true
 maxDistance=2250
 soundClassObject=&Fort_WeaponFoley
 volumeMultiplier=0.501
 concurrencySet [/Game/Sounds/Concurrency_Rules/Bow_Global_Conc, /Game/Sounds/Concurrency_Rules/Bow_Local_Conc]
}
