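# Sound cue for the looping in-air sound of a charged Shockwave Bow projectile.
# A looping wave player feeds a default Modulator, then an Enveloper whose cubic volume curve starts reduced, dips, then climbs to full.
# It uses the arrow in-air loop attenuation, the ranged projectile effects sound class, and global and local bow concurrency rules.
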SoundCue /PrimalGameplay/Sounds/Weapons/Bows/Shockwave/Bow_Shockwave_Prj_InAir_Charged_Loop_Cue
root=Enveloper_0
nodes [
 {
  id=Enveloper_0
  children [Modulator_0]
  properties {
   pitchCurve {editorCurveData {keys [{time=0}]}}
   volumeCurve {editorCurveData {keys [{arriveTangent=-1.5962, interpMode=RCIM_Cubic, leaveTangent=-1.5962, tangentMode=RCTM_User, time=0, value=0.6}, <at 1 1>]}}
  }
 }
 {id=Modulator_0, children [WavePlayer_0]}
 {
  id=WavePlayer_0
  properties {bLooping=true, soundWaveAssetPtr=&/Bow_Shockwave_Prj_InAir_Charged_Loop}
 }
]
properties {
 attenuationSettings=/Game/Sounds/Attenuation/Fort_Weapon_Projectile_Arrow_InAirLoop_Att
 maxDistance=16800
 soundClassObject=&Fort_WeaponRanged_Projectile_Effects
 volumeMultiplier=1
 attenuationOverrides {falloffDistance=5000, reverbWetLevelMax=0.7, reverbWetLevelMin=0.2, attenuationShapeExtents {x=1000}}
 concurrencySet [/Game/Sounds/Concurrency_Rules/Bow_Global_Conc, /Game/Sounds/Concurrency_Rules/Bow_Local_Conc]
}
